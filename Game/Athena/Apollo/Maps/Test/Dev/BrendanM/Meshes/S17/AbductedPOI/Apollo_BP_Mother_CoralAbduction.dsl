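; Coral abduction Blueprint for the Mother POI, with helpers that place building foundations along a spline at a given progress.
; MasterTransform samples the spline, snaps to the build grid, then sets and enables the dynamic foundation transform; preview variants load editor preview levels instead.
; SnaptoGridandRotate also computes a 90-degree-snapped yaw, but only location feeds the transform.
; OnRep_CoralProgress calls Coral_LoadandTransform with zero, but that event, other OnReps and construction script are empty.
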(event Custom|Coral_LoadandTransform
  (params (Progress :type Float%20%28double-precision%29)))

(fn MasterTransform
  (params (Spline :type Spline%20Component%20Object%20Reference) (Foundation :type Building%20Foundation%20Object%20Reference) (Progress :type Float%20%28double-precision%29))
  (CallFunction|MoveAndLoadFoundation (Math|Conversions|ToTransform(Vector) (CallFunction|SnapToGridAndRotate (Spline|GetLocationatDistanceAlongSpline Spline (* (Spline|GetSplineLength Spline) Progress) "World"))) Foundation))

(fn MasterTransform_Preview
  (params (Spline :type Spline%20Component%20Object%20Reference) (Foundation :type Building%20Foundation%20Object%20Reference) (Progress :type Float%20%28double-precision%29))
  (Default|MoveAndLoadFoundationPreview (Math|Conversions|ToTransform(Vector) (CallFunction|SnapToGridAndRotate (Spline|GetLocationatDistanceAlongSpline Spline (* (Spline|GetSplineLength Spline) Progress) "World"))) Foundation))

(fn MoveandLoadFoundation
  (params (Location :type Transform) (Foundation :type Building%20Foundation%20Object%20Reference))
  (Class|BuildingFoundation|SetDynamicFoundationTransform Foundation Location)
  (Class|BuildingFoundation|SetDynamicFoundationEnabled Foundation true))

(fn MoveandLoadFoundation_Preview
  (params (Location :type Transform) (Foundation :type Building%20Foundation%20Object%20Reference))
  (Class|BuildingFoundation|EditorOnlyAddTransformtoPreviewLevel Foundation Location)
  (Class|BuildingFoundation|EditorOnlyLoadPreviewLevel Foundation false))

(fn OnRep_CoralChunk_01)

(fn OnRep_CoralChunk_02)

(fn OnRep_CoralChunk_03)

(fn OnRep_CoralChunk_04)

(fn OnRep_CoralProgress
  (CallFunction|CoralLoadandTransform 0.0))

(fn OnRep_StageDebug)

(fn SnaptoGridandRotate
  (params (Location :type Vector))
  (return (Fortnite|SnaptoFortBuildingGrid Location) (Math|Rotator|MakeRotator 0.0 0.0 (Math|Float|SnapToGrid(Float) (.yaw (Math|Conversions|RotationFromXVector Location)) 9e+01))))

(fn ConstructionScript)
